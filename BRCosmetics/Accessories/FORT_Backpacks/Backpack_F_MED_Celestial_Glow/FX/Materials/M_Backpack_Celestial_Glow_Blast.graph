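# Unlit, alpha-composite Niagara sprite material for a radial blast effect on the Celestial Glow backpack.
# A reflected gradient band travels along the radius over particle lifetime.
# Noise in radial coordinates, randomized per particle, pans with a decelerating curve, darkens toward the center and erodes through a sphere mask.
# Band times noise is sharpened, depth-faded and tinted by particle color, squared for the emissive adjustment function.
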Material /BRCosmetics/Accessories/FORT_Backpacks/Backpack_F_MED_Celestial_Glow/FX/Materials/M_Backpack_Celestial_Glow_Blast
graph {
 properties {
  bUsedWithNiagaraSprites=true
  blendMode=BLEND_AlphaComposite
  pixelDepthOffsetMode=PDOM_Legacy
  shadingModel=MSM_Unlit
  translucencyPass=MTP_BeforeDOF
  usedShadingModels=Unlit
 }
 nodes [
  n1 = MaterialFunctionCall() {materialFunction=&VectorToRadialValue}
  n2 = MaterialFunctionCall!(
   Gradient=MaterialFunctionCall!(Gradient=ComponentMask(None=@n1.Radial Coordinates) {g=true}, Mask Position=ParticleRelativeTime(), Mask Size=Constant() {r=0.4}) {materialFunction=/Game/Effects/Fort_Effects/MaterialFunctions/MF_GradientMask}.Linear
   Center Point=Constant() {r=0.8}
  ) {materialFunction=/Game/Effects/Fort_Effects/MaterialFunctions/MF_GradientReflected}
  n3 = ParticleRandom()
  n4 = Power(Base=OneMinus(None=ParticleRelativeTime()), Exp=ScalarParameter() {defaultValue=2, parameterName=Pan Exponent}) {constExponent=3}
  n5 = Constant()
  n6 = ScalarParameter() {defaultValue=-1.5, parameterName=Pan Amount}
  n7 = TextureCoordinate()
  n8 = TextureSample!(
   UVs=Panner(Coordinate=Add(A=@n1.Radial Coordinates, B=@n3), Time=OneMinus(None=@n4), Speed=AppendVector(A=@n5, B=@n6)) {speedY=-3}
   Tex
   "DDX(UVs)"=DDX(Value=@n7)
   "DDY(UVs)"=DDY(Value=@n7)
   Apply View MipBias
  ) {
   mipValueMode=TMVM_Derivative
   samplerType=SAMPLERTYPE_LinearGrayscale
   texture=/BRCosmetics/Athena/Effects/Particles/Skydiving/NS_Contrail_DarkAndLight/T_DarkAndLight_Noise
  }
  n9 = OneMinus(None=@n8.R)
  n10 = MaterialFunctionCall(Vector 2=Subtract(A=@n7) {constB=0.5}) {materialFunction=/Engine/Functions/Engine_MaterialFunctions02/Utility/VectorLength}
  n11 = Power(Base=Saturate(None=Multiply(A=@n10.V2 Length) {constB=2})) {constExponent=0.333}
  n12 = MaterialFunctionCall!(
   Gradient=OneMinus(None=Multiply(A=@n9, B=@n11))
   Erosion Amount=OneMinus(None=MaterialFunctionCall!(UV, Size, Curvature) {materialFunction=/Game/Effects/Fort_Effects/MaterialFunctions/MF_Sphere}.Height)
   Softness=ScalarParameter() {defaultValue=1, parameterName=Softness, sliderMax=1}
  ) {materialFunction=/BRCosmetics/Athena/Effects/Particles/Emotes/NS_Emote_Spectrum/Functions/MF_Erosion}
  n13 = DepthFade(
   Opacity=Power(Base=Multiply(A=@n2.Result, B=@n12.Default Range), Exp=ScalarParameter() {defaultValue=3, parameterName=Exponent})
   FadeDistance=ScalarParameter() {defaultValue=33, parameterName=Depth Fade Distance}
  )
  n14 = ParticleColor()
  MaterialFunctionCall(Emissive=Multiply(A=@n13, B=Multiply(A=@n14.RGB, B=@n14.RGB))) {materialFunction=&MF_FX_EmissiveAdjustments}
  Multiply(A=@n13, B=@n14.RGB)
 ]
}
