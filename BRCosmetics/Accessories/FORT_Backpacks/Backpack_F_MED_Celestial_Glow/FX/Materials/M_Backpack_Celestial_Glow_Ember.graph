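# Additive unlit Niagara particle material for the ember effect on the Celestial Glow backpack cosmetic, drawing a soft round glow without textures.
# Particle color gets a per-particle random hue shift of up to half a cycle, then passes through MF_FX_EmissiveAdjustments.
# A UV-distance radial gradient, boosted by Center Glow Size and shaped by an exponent, multiplies particle alpha, with optional depth fade through a static switch.
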Material /BRCosmetics/Accessories/FORT_Backpacks/Backpack_F_MED_Celestial_Glow/FX/Materials/M_Backpack_Celestial_Glow_Ember
graph {
 properties {
  bUsedWithNiagaraMeshParticles=true
  bUsedWithNiagaraSprites=true
  blendMode=BLEND_Additive
  pixelDepthOffsetMode=PDOM_Legacy
  shadingModel=MSM_Unlit
  usedShadingModels=Unlit
 }
 nodes [
  n1 = Multiply(A=ParticleRandom()) {constB=0.5}
  n2 = ParticleColor()
  Reroute(None=MaterialFunctionCall(Emissive=MaterialFunctionCall(Hue Shift Percentage=@n1, Texture=@n2.RGB) {materialFunction=&HueShift}.Result) {materialFunction=&MF_FX_EmissiveAdjustments}.Result)
  n3 = Distance(A=TextureCoordinate(), B=Constant() {r=0.5})
  n4 = OneMinus(None=Saturate(None=Multiply(A=@n3) {constB=2}))
  n5 = OneMinus(None=Saturate(None=ScalarParameter() {defaultValue=0.3, parameterName=Center Glow Size}))
  n6 = Multiply(A=@n2.A, B=Power(Base=Divide(A=@n4, B=@n5), Exp=ScalarParameter() {defaultValue=3, parameterName=Exponent}))
  StaticSwitchParameter(True=DepthFade(Opacity=@n6, FadeDistance=ScalarParameter() {group=Opacity, parameterName=Depth Fade Distance}), False=@n6) {group=Opacity, parameterName=Enable Depth Fade}
 ]
}
