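# Unlit alpha-composite particle material for the glowing hands effect on the Celestial Glow backpack.
# Fixed flipbook frames, one flipped and offset, crossfade through contrast-sharpened lerps in a time loop offset by particle alpha.
# The shape is masked by facing angle and UV falloff, then tinted with rainbow colors driven by panning sine noise.
# OnKillMaterialBlend swaps in another frame and boosts intensity, while particle red feeds depth-fade opacity.
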Material /BRCosmetics/Accessories/FORT_Backpacks/Backpack_F_MED_Celestial_Glow/FX/Materials/M_Backpack_Celestial_Glow_Hands
graph {
 properties {
  bUsedWithNiagaraMeshParticles=true
  bUsedWithNiagaraSprites=true
  blendMode=BLEND_AlphaComposite
  pixelDepthOffsetMode=PDOM_Legacy
  shadingModel=MSM_Unlit
  twoSided=true
  usedShadingModels=Unlit
 }
 nodes [
  n1 = Abs(None=DotProduct(A=CameraVectorWS(), B=VertexNormalWS()))
  n2 = ScalarParameter() {defaultValue=1, parameterName=Fresnel Exponent}
  n3 = TextureSample(UVs=MaterialFunctionCall("Animation  Phase (0-1)"=Constant() {r=0.625}) {materialFunction=&FlipBook}.UVs) {
   samplerSource=SSM_Wrap_WorldGroupSettings
   samplerType=SAMPLERTYPE_LinearGrayscale
   texture=/BRCosmetics/Accessories/FORT_Backpacks/Backpack_F_MED_Celestial_Glow/FX/Textures/T_Backpack_Celestial_Glow_Flipbook
  }
  n4 = TextureSample(UVs=MaterialFunctionCall("Animation  Phase (0-1)"=Constant() {r=0.125}) {materialFunction=&FlipBook}.UVs) {
   samplerSource=SSM_Wrap_WorldGroupSettings
   samplerType=SAMPLERTYPE_LinearGrayscale
   texture=&T_Backpack_Celestial_Glow_Flipbook
  }
  n5 = ParticleColor()
  n6 = Add(A=Time(), B=Multiply(A=OneMinus(None=@n5.A), B=ScalarParameter() {defaultValue=1, parameterName=Frame Offset per Particle}))
  n7 = ScalarParameter() {defaultValue=0.05, parameterName=Animation Speed}
  n8 = Multiply(A=Frac(None=Multiply(A=@n6, B=@n7))) {constB=4}
  n9 = ScalarParameter() {defaultValue=3, parameterName=Contrast}
  n10 = LinearInterpolate(
   A=@n3.B
   B=@n4.R
   Alpha=MaterialFunctionCall(In=Saturate(None=@n8), Contrast=@n9) {materialFunction=&CheapContrast}.Result
  )
  n11 = TextureSample(UVs=MaterialFunctionCall("Animation  Phase (0-1)"=Constant() {r=0.375}) {materialFunction=&FlipBook}.UVs) {
   samplerSource=SSM_Wrap_WorldGroupSettings
   samplerType=SAMPLERTYPE_LinearGrayscale
   texture=&T_Backpack_Celestial_Glow_Flipbook
  }
  n12 = MaterialFunctionCall(In=Saturate(None=Subtract(A=@n8)), Contrast=@n9) {materialFunction=&CheapContrast}
  n13 = OneMinus(None=TextureCoordinate())
  n14 = Constant2Vector() {g=-0.0762}
  n15 = TextureSample(UVs=MaterialFunctionCall("Animation  Phase (0-1)"=Constant() {r=0.375}, UVs=Add(A=@n13, B=@n14)) {materialFunction=&FlipBook}.UVs) {
   samplerSource=SSM_Wrap_WorldGroupSettings
   samplerType=SAMPLERTYPE_LinearGrayscale
   texture=&T_Backpack_Celestial_Glow_Flipbook
  }
  n16 = Saturate(None=Subtract(A=@n8) {constB=2})
  n17 = Subtract(A=@n8) {constB=3}
  n18 = LinearInterpolate(
   A=LinearInterpolate(
    A=LinearInterpolate(A=@n10, B=@n11.G, Alpha=@n12.Result)
    B=@n15.G
    Alpha=MaterialFunctionCall(In=@n16, Contrast=@n9) {materialFunction=&CheapContrast}.Result
   )
   B=@n3.B
   Alpha=MaterialFunctionCall(In=Saturate(None=@n17), Contrast=@n9) {materialFunction=&CheapContrast}.Result
  )
  n19 = TextureSample(UVs=MaterialFunctionCall("Animation  Phase (0-1)"=Constant() {r=0.875}) {materialFunction=&FlipBook}.UVs) {
   samplerSource=SSM_Wrap_WorldGroupSettings
   samplerType=SAMPLERTYPE_LinearGrayscale
   texture=&T_Backpack_Celestial_Glow_Flipbook
  }
  n20 = DynamicParameter() {paramNames [OnKillMaterialBlend, - unused -, - unused -, - unused -]}
  n21 = Power(
   Base=Saturate(None=LinearInterpolate(A=@n18, B=@n19.A, Alpha=@n20.OnKillMaterialBlend))
   Exp=ScalarParameter() {defaultValue=4, parameterName=Exponent}
  )
  n22 = Divide(A=TextureCoordinate())
  n23 = MaterialFunctionCall(Float2=SquareRoot(None=Sine(None=@n22))) {materialFunction=&BreakOutFloat2Components}
  n24 = Multiply(A=@n23.R, B=@n23.R)
  n25 = Panner(Time=@n6, Speed=AppendVector(A=Constant(), B=ScalarParameter() {defaultValue=1, parameterName=Noise Panner Speed})) {bFractionalPart=true}
  n26 = Add(
   A=SquareRoot(None=TextureSample(UVs=@n25) {texture=/Game/Global/Textures/Default/Masks_and_Gradients/T_BlasterBeamDistort}.R)
   B=Multiply(A=@n6, B=ScalarParameter() {defaultValue=1, parameterName=Noise Sine Speed})
  )
  n27 = ConstantBiasScale(None=Sine(None=Divide(A=@n26) {constB=4}))
  n28 = Multiply(
   A=Multiply(A=Power(Base=@n1, Exp=@n2), B=Multiply(A=@n21, B=@n24))
   B=Multiply(
    A=MaterialFunctionCall!(Gradient=@n27) {materialFunction=/BRCosmetics/Athena/Effects/Particles/Skydiving/NS_Contrail_DarkAndLight/MF_DarkAndLight_RainbowColors}.Result
    B=ScalarParameter() {defaultValue=44, parameterName=Boost}
   ) {constB=3}
  )
  n29 = ScalarParameter() {defaultValue=11, parameterName=Reactivity Intensity Boost}
  n30 = Multiply(
   A=DepthFade(Opacity=ParticleColor().R, FadeDistance=ScalarParameter() {defaultValue=22, parameterName=Depth Fade Distance})
   B=LinearInterpolate(A=@n28, B=Multiply(A=@n28, B=@n29), Alpha=@n20.OnKillMaterialBlend)
  )
  ComponentMask(None=@n30) {a=true}
  MaterialFunctionCall(Emissive=ComponentMask(None=@n30) {b=true, g=true, r=true}) {materialFunction=&MF_FX_EmissiveAdjustments}
  ScalarParameter() {parameterName=Reactivity, sliderMax=1}
 ]
}
