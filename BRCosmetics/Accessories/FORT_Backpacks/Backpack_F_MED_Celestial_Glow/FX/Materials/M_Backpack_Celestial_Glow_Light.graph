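# Unlit alpha-composite Niagara sprite material for the glowing light effect on the Celestial Glow backpack.
# A soft sphere mask is overlay-blended with panning polar-UV noise, sharpened, depth-faded and tinted by squared particle color.
# The sharpened mask also drives a hue shift and divide that intensify the core.
# Dynamic parameter OnKillMaterialBlend blends in a brightness boost before clamped emissive adjustment, and a camera-facing offset vector is computed.
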Material /BRCosmetics/Accessories/FORT_Backpacks/Backpack_F_MED_Celestial_Glow/FX/Materials/M_Backpack_Celestial_Glow_Light
graph {
 properties {
  bUsedWithNiagaraSprites=true
  blendMode=BLEND_AlphaComposite
  pixelDepthOffsetMode=PDOM_Legacy
  shadingModel=MSM_Unlit
  usedShadingModels=Unlit
 }
 nodes [
  n1 = ParticleColor()
  AppendVector(A=@n1.RGB, B=@n1.RGB)
  n2 = Constant() {r=0.5}
  n3 = SphereMask(A=TextureCoordinate(), B=@n2, Radius=@n2) {hardnessPercent=0}
  n4 = MaterialFunctionCall!(
   UV
   Twist Center=ScalarParameter() {defaultValue=-0.222, parameterName=Twist Center}
   Twist Edges=ScalarParameter() {parameterName=Twist Edges}
  ) {materialFunction=/Game/Accessories/FORT_Backpacks/Backpack_F_MED_Celestial_Glow/FX/Materials/MF_PolarUV}
  n5 = Constant2Vector() {g=0.5, r=1}
  n6 = Time()
  n7 = ScalarParameter() {defaultValue=0.5, parameterName=Panner Speed}
  n8 = TextureSample!(
   UVs=Panner(Coordinate=Multiply(A=@n4.Polar UVs, B=@n5), Time=Multiply(A=@n6, B=@n7)) {bFractionalPart=true, speedX=-0.2, speedY=-0.4}
   Tex
   "DDX(UVs)"=@n4.DDX
   "DDY(UVs)"=@n4.DDX
   Apply View MipBias
  ) {
   mipValueMode=TMVM_Derivative
   samplerType=SAMPLERTYPE_LinearGrayscale
   texture=/BRCosmetics/Athena/Effects/Particles/Skydiving/NS_Contrail_DarkAndLight/T_DarkAndLight_Noise
  }
  n9 = Constant() {r=4}
  n10 = Power(
   Base=ComponentMask(None=MaterialFunctionCall(Base=@n3, Blend=@n8.R) {materialFunction=&Blend_Overlay}.Result) {r=true}
   Exp=@n9
  )
  n11 = Multiply(
   A=DepthFade(Opacity=@n10, FadeDistance=ScalarParameter() {defaultValue=11, parameterName=Depth Fade Distance})
   B=Multiply(A=@n1.RGB, B=@n1.RGB)
  )
  ComponentMask(None=@n11) {a=true}
  Constant() {r=1}
  n12 = Power(Base=@n10, Exp=@n9)
  n13 = Divide(
   A=MaterialFunctionCall(Hue Shift Percentage=Multiply(A=@n12) {constB=0.5}, Texture=ComponentMask(None=@n11) {b=true, g=true, r=true}) {materialFunction=&HueShift}.Result
   B=OneMinus(None=@n12)
  )
  n14 = Multiply(A=@n13, B=ScalarParameter() {defaultValue=2, parameterName=OnKill Brightness Boost})
  n15 = DynamicParameter() {paramNames [OnKillMaterialBlend, - unused -, - unused -, - unused -]}
  MaterialFunctionCall(Emissive=Clamp(None=LinearInterpolate(A=@n13, B=@n14, Alpha=@n15.OnKillMaterialBlend)) {clampMode=CMODE_ClampMax, maxDefault=111}) {materialFunction=&MF_FX_EmissiveAdjustments}
  n16 = CameraPositionWS()
  n17 = WorldPosition()
  Multiply(A=Normalize(VectorInput=Subtract(A=@n16, B=@n17.XYZ)), B=ScalarParameter() {defaultValue=11, parameterName=Camera Offset})
 ]
}
