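# Niagara effect for the Celestial Glow backpack cosmetic, with a persistent idle look and an on-kill burst.
# User.OnKillBool gates spawn probability: false spawns Center_Glow and Hands, true spawns the Blast flash and outward-flying Embers.
# Hands places five rotationally offset mesh particles through two mesh renderers and spins them.
# User.OnKillMaterialBlend feeds the glow and hand materials and raises the hands' spin rate.
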NiagaraSystem /BRCosmetics/Accessories/FORT_Backpacks/Backpack_F_MED_Celestial_Glow/FX/NS_Backpack_Celestial_Glow
dependencies {usedDataInterfaces [&NiagaraDataInterfaceCurve], usedDynamicInputs [], usedModules [&SystemState]}
emitters {
 Blast {
  emitterUpdateScript [
   &EmitterState()
   &SpawnBurst_Instantaneous(
    Spawn Probability=&MakeCustomFloatFromBool(Bool=$User.OnKillBool, True Float=1, False Float=0)
    Use Spawn Probability=true
    UseLoopCountLimit=false
   )
  ]
  particleSpawnScript [&InitializeParticle(
    Color=<rgba 1.3333 0 2 11>
    Lifetime=0.85
    Position Offset=<0 -11 0>
    Position Offset Coordinate Space=::Local
    Sprite Size Mode=::NewEnumerator3
    Uniform Sprite Size=55
    UsePositionOffset=true
    Write Material Random=true
   )]
  particleUpdateScript [
   &ParticleState()
   &ScaleColor(
    Scale Alpha=&FloatFromCurve(FloatCurve=DataInterface() {propertyValues {Curve {keys [{arriveTangent=-99.1033, interpMode=RCIM_Cubic, leaveTangent=-99.1033, tangentMode=RCTM_User, time=0, value=33}, <at 1 1>]}}})
    ScaleA=true
    ScaleRGB=true
   )
  ]
  rendererClasses [&NiagaraSpriteRendererProperties]
  renderers [{rendererClass=&NiagaraSpriteRendererProperties}]
 }
 Center_Glow {
  emitterUpdateScript [
   &EmitterState(
    Inactive Response=::NewEnumerator1
    Life Cycle Mode=::NewEnumerator1
    Loop Behavior=::NewEnumerator1
    Loop Duration Mode=::NewEnumerator1
    UseLoopDelay=false
   )
   &SpawnBurst_Instantaneous(
    Spawn Probability=&MakeCustomFloatFromBool(Bool=$User.OnKillBool, True Float=0, False Float=1)
    Use Spawn Probability=true
    UseLoopCountLimit=false
   )
  ]
  particleSpawnScript [&InitializeParticle(
    Color=<rgba 0.222 0 0.333 1>
    Lifetime=5
    Position Offset=<0 -8 0>
    Sprite Rotation Angle=/Game/Effects/Niagara/DynamicInputs/Math/FN_NormalizedParticleCount()
    Sprite Rotation Mode=::NewEnumerator1
    Sprite Size Mode=::NewEnumerator3
    Uniform Sprite Size=66
    UsePositionOffset=true
    Write Material Random=true
   )]
  particleUpdateScript [
   &ParticleState(
    Kill Particles When Lifetime Has Elapsed=false
    Let Infinitely Lived Particles Die When Emitter Deactivates=true
    Loop Particles Lifetime=true
   )
   &DynamicMaterialParameters(
    Index 0 Param 1=$User.OnKillMaterialBlend
    Module.Param0WriteEnabled=true
    Module.Param1WriteEnabled=true
    Module.Param2WriteEnabled=true
    Module.Param3WriteEnabled=true
   )
  ]
  rendererClasses [&NiagaraSpriteRendererProperties]
  renderers [{rendererClass=&NiagaraSpriteRendererProperties}]
 }
 Embers {
  emitterUpdateScript [
   &EmitterState()
   &SpawnBurst_Instantaneous(
    Spawn Count=99
    Spawn Probability=&MakeCustomFloatFromBool(Bool=$User.OnKillBool, True Float=1, False Float=0)
    Use Spawn Probability=true
    UseLoopCountLimit=false
   )
  ]
  particleSpawnScript [
   &InitializeParticle(
    Color=<rgba 0.6667 0 1 1>
    Lifetime Max=1.45
    Lifetime Min=0.35
    Lifetime Mode=::NewEnumerator1
    Position Mode=::NewEnumerator0
    Sprite Size Mode=::NewEnumerator1
    Uniform Sprite Size Max=1.3
    Uniform Sprite Size Min=0.7
    Write Material Random=true
   )
   &SphereLocation(
    MaskSpawnGroup=false
    Offset=<0 -11 0>
    Override Seed=false
    Sphere Radius=1
    Surface Only=false
    UseNonUniformScale=false
    UseOffset=true
    UseSphereOrientation=false
   )
   &AddVelocityFromPoint(
    Module.Use Falloff Distance=false
    Origin Offset=<0 -11 0>
    Velocity Strength=&RandomRangeFloat(Minimum=77, Maximum=222)
   )
  ]
  particleUpdateScript [
   &Drag(Drag=3, Use Linear Drag=true, Use Rotational Drag=false)
   &SolveForcesAndVelocity(Clamp Velocity=false, Limit Acceleration=false)
   &ParticleState()
   &ScaleColor(
    Scale Alpha=&FloatFromCurve(FloatCurve=DataInterface() {propertyValues {Curve {keys [
         {arriveTangent=-1.322, interpMode=RCIM_Cubic, leaveTangent=-1.322, tangentMode=RCTM_User, time=0, value=1}
         {arriveTangent=7.0277, interpMode=RCIM_Cubic, leaveTangent=7.0277, tangentMode=RCTM_User, time=0.5, value=0.45}
         <at 1 0>
        ]}}})
    ScaleA=true
    ScaleRGB=true
   )
  ]
  rendererClasses [&NiagaraSpriteRendererProperties]
  renderers [{rendererClass=&NiagaraSpriteRendererProperties}]
 }
 Hands {
  emitterUpdateScript [
   &EmitterState(
    Inactive Response=::NewEnumerator1
    Life Cycle Mode=::NewEnumerator1
    Loop Behavior=::NewEnumerator1
    Loop Duration Mode=::NewEnumerator1
    UseLoopDelay=false
   )
   &SpawnBurst_Instantaneous(
    Spawn Count=5
    Spawn Probability=&MakeCustomFloatFromBool(Bool=$User.OnKillBool, True Float=0, False Float=1)
    Use Spawn Probability=true
    UseLoopCountLimit=false
   )
  ]
  particleSpawnScript [
   &InitializeParticle(
    Color=&MakeLinearColorFromVectorAndFloat(
     "Vector (RGB)"=<1 1 1>
     "Float (Alpha)"=&Divide_Float(A=&MakeFloatFromInt(Int=&ReturnParticleIDAttributes()), B=5)
    )
    Lifetime=5
    Mesh Scale Mode=::NewEnumerator3
    Mesh Uniform Scale=0.75
    Position Offset=<0 -5 0>
    Sprite Rotation Angle=&FN_NormalizedParticleCount()
    Sprite Rotation Mode=::NewEnumerator1
    UsePositionOffset=true
    Write Material Random=true
   )
   &InitialMeshOrientation(
    Orientation Axis=<0 1 0>
    Orientation Coordinate Space=::Local
    Rotation=&MakeVector(
     X=&Divide_Float(A=&MakeFloatFromInt(Int=&ReturnParticleIDAttributes()), B=5)
     Y=0
     Z=0
    )
    Use Orientation Vector=true
    Use Rotation Vector=true
   )
   /SetVariables(
    Particles.OffsetColor:LinearColor=&MultiplyLinearColor(Linear Color=$Particles.Color, Scale Factor=<0.5 0 0 1>)
    Particles.OffsetPosition:Vector3f=&Add_Vector(A=Linked() {linkedVariable {name=Particles.Position, type=NiagaraPosition}}, B=<0 -0.5 0>)
   ) {isSetParameters=true}
  ]
  particleUpdateScript [
   &ParticleState(
    Kill Particles When Lifetime Has Elapsed=false
    Let Infinitely Lived Particles Die When Emitter Deactivates=true
    Loop Particles Lifetime=true
   )
   &UpdateMeshOrientation(
    Rotation Coordinate Space=::NewEnumerator2
    Rotation Rate=&Lerp_Float(A=0.1, B=1, Alpha=$User.OnKillMaterialBlend)
    Rotation Vector=<0 -1 0>
   )
   &DynamicMaterialParameters(
    Index 0 Param 1=$User.OnKillMaterialBlend
    Module.Param0WriteEnabled=true
    Module.Param1WriteEnabled=true
    Module.Param2WriteEnabled=true
    Module.Param3WriteEnabled=true
   )
  ]
  rendererClasses [&NiagaraMeshRendererProperties]
  renderers [
   {rendererClass=&NiagaraMeshRendererProperties, rendererIndex=0}
   {rendererClass=&NiagaraMeshRendererProperties, rendererIndex=1}
  ]
 }
}
summary {
 systemName=NS_Backpack_Celestial_Glow
 emitters [
  {emitterName=Hands, rendererClasses [&NiagaraMeshRendererProperties]}
  {emitterName=Center_Glow, rendererClasses [&NiagaraSpriteRendererProperties]}
  {emitterName=Blast, rendererClasses [&NiagaraSpriteRendererProperties]}
  {emitterName=Embers, rendererClasses [&NiagaraSpriteRendererProperties]}
 ]
 userVariables [
  User.OnKillBool {type=NiagaraBool}
  User.OnKillMaterialBlend {type=NiagaraFloat, defaultValue=0, description=""}
 ]
}
